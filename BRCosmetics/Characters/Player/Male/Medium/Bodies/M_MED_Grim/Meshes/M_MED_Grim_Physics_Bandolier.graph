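# Physics asset for the bandolier bullets on the M_MED_Grim player cosmetic body, covering five dynamic bullet bones.
# Each bullet has an offset capsule body paired with a tiny capsule on its matching "_mstr" bone.
# Five identical constraints tie each bullet to its master with tight swing and twist limits, so bullets can only wobble slightly.
# No constraints link the bullets to one another.
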PhysicsAsset /BRCosmetics/Characters/Player/Male/Medium/Bodies/M_MED_Grim/Meshes/M_MED_Grim_Physics_Bandolier
bodies {
 dyn_bandolierBullet_01 [{shapeType=Capsule, length=7, radius=1.6, center=<1.5807 0.0262 0.7815>}]
 dyn_bandolierBullet_01_mstr [{shapeType=Capsule, length=0.505, radius=0.505}]
 dyn_bandolierBullet_02 [{shapeType=Capsule, length=7, radius=1.6, center=<1.5807 0.0392 0.7809>}]
 dyn_bandolierBullet_02_mstr [{shapeType=Capsule, length=0.505, radius=0.505}]
 dyn_bandolierBullet_03 [{shapeType=Capsule, length=7, radius=1.6, center=<1.5807 0.0392 0.7809>}]
 dyn_bandolierBullet_03_mstr [{shapeType=Capsule, length=0.505, radius=0.505}]
 dyn_bandolierBullet_04 [{shapeType=Capsule, length=7, radius=1.6, center=<1.5807 0.0392 0.7809>}]
 dyn_bandolierBullet_04_mstr [{shapeType=Capsule, length=0.505, radius=0.505}]
 dyn_bandolierBullet_05 [{shapeType=Capsule, length=7, radius=1.6, center=<1.5807 0.0392 0.7809>}]
 dyn_bandolierBullet_05_mstr [{shapeType=Capsule, length=0.505, radius=0.505}]
}
constraints [
 {bone1Name=dyn_bandolierBullet_01, bone2Name=dyn_bandolierBullet_01_mstr, swing1LimitDegrees=8, swing2LimitDegrees=8, twistLimitDegrees=3}
 {bone1Name=dyn_bandolierBullet_02, bone2Name=dyn_bandolierBullet_02_mstr, swing1LimitDegrees=8, swing2LimitDegrees=8, twistLimitDegrees=3}
 {bone1Name=dyn_bandolierBullet_03, bone2Name=dyn_bandolierBullet_03_mstr, swing1LimitDegrees=8, swing2LimitDegrees=8, twistLimitDegrees=3}
 {bone1Name=dyn_bandolierBullet_04, bone2Name=dyn_bandolierBullet_04_mstr, swing1LimitDegrees=8, swing2LimitDegrees=8, twistLimitDegrees=3}
 {bone1Name=dyn_bandolierBullet_05, bone2Name=dyn_bandolierBullet_05_mstr, swing1LimitDegrees=8, swing2LimitDegrees=8, twistLimitDegrees=3}
]
