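# Physics asset for the male medium Grim player body that covers its eight belt pouch bones, four per side.
# Each dyn_pouch bone has a mirrored capsule body, and the pelvis carries a tiny placeholder capsule as the shared parent.
# Every pouch is constrained directly to the pelvis with tight angular limits, slightly looser on one swing axis, allowing only a small jiggle.
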PhysicsAsset /BRCosmetics/Characters/Player/Male/Medium/Bodies/M_MED_Grim/Meshes/M_MED_Grim_Physics_Pouches
bodies {
 dyn_pouch_01_l [{shapeType=Capsule, length=5.9482, radius=4.1472, center=<2.4699 -0.4562 -2.7423>}]
 dyn_pouch_01_r [{shapeType=Capsule, length=5.9482, radius=4.1472, center=<2.4699 0.4562 -2.7423>}]
 dyn_pouch_02_l [{shapeType=Capsule, length=5.9689, radius=4.248, center=<2.4938 -0.5146 -2.731>}]
 dyn_pouch_02_r [{shapeType=Capsule, length=5.9689, radius=4.248, center=<2.4938 0.5146 -2.7309>}]
 dyn_pouch_03_l [{shapeType=Capsule, length=5.9418, radius=4.1189, center=<2.4929 -0.4401 -2.7452>}]
 dyn_pouch_03_r [{shapeType=Capsule, length=5.9418, radius=4.1189, center=<2.4929 0.4401 -2.7452>}]
 dyn_pouch_04_l [{shapeType=Capsule, length=5.94, radius=4.1112, center=<2.4805 -0.4356 -2.7459>}]
 dyn_pouch_04_r [{shapeType=Capsule, length=5.94, radius=4.1112, center=<2.4805 0.4356 -2.7459>}]
 pelvis [{shapeType=Capsule, length=1, radius=1}]
}
constraints [
 {bone1Name=dyn_pouch_01_l, bone2Name=pelvis, swing1LimitDegrees=10, swing2LimitDegrees=2, twistLimitDegrees=2}
 {bone1Name=dyn_pouch_02_l, bone2Name=pelvis, swing1LimitDegrees=10, swing2LimitDegrees=2, twistLimitDegrees=2}
 {bone1Name=dyn_pouch_03_l, bone2Name=pelvis, swing1LimitDegrees=10, swing2LimitDegrees=2, twistLimitDegrees=2}
 {bone1Name=dyn_pouch_04_l, bone2Name=pelvis, swing1LimitDegrees=10, swing2LimitDegrees=2, twistLimitDegrees=2}
 {bone1Name=dyn_pouch_04_r, bone2Name=pelvis, swing1LimitDegrees=10, swing2LimitDegrees=2, twistLimitDegrees=2}
 {bone1Name=dyn_pouch_03_r, bone2Name=pelvis, swing1LimitDegrees=10, swing2LimitDegrees=2, twistLimitDegrees=2}
 {bone1Name=dyn_pouch_02_r, bone2Name=pelvis, swing1LimitDegrees=10, swing2LimitDegrees=2, twistLimitDegrees=2}
 {bone1Name=dyn_pouch_01_r, bone2Name=pelvis, swing1LimitDegrees=10, swing2LimitDegrees=2, twistLimitDegrees=2}
]
