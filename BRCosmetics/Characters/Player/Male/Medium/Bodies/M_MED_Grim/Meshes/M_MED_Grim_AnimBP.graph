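# Post-process style AnimBlueprint for the Grim male medium cosmetic body, adding deformation and secondary motion to the pose copied from its parent mesh.
# Control Rigs, weighted bandolier-bullet constraints and a right shoulderpad offset, enabled while aiming a pistol, feed a cached pre-simulation pose.
# Two chained RigidBody nodes then simulate pouches and bandolier using gravity from a deprecated override state function.
# Arm-rotation shoulderpad nodes sit unconnected to the output.
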AnimBlueprint /BRCosmetics/Characters/Player/Male/Medium/Bodies/M_MED_Grim/Meshes/M_MED_Grim_AnimBP
target_skeleton=&/M_MED_Grim_Skeleton
graphs [
 AnimGraph {nodes [
   n1 = "Transform (Modify) Bone - Bone: dyn_shoulderpad_mstr_l"(Alpha=1, Translation="0,0,0", Rotation="5,0,0", Scale="1,1,1") {
    rotationMode=BMM_Additive
    rotationSpace=BCS_BoneSpace
    translationMode=BMM_Additive
    translationSpace=BCS_BoneSpace
    alphaBoolBlend {blendOption=Linear}
    alphaScaleBiasClamp {bInterpResult=true, interpSpeedDecreasing=20, interpSpeedIncreasing=20}
    boneToModify {boneName=dyn_shoulderpad_mstr_l}
    rotation {pitch=5}
   }
   "Transform (Modify) Bone - Bone: dyn_shoulderpad_mstr_r"(ComponentPose=@n1, Alpha=1, Translation="0,0,0", Rotation="0,0,0", Scale="1,1,1") {
    rotationMode=BMM_Additive
    rotationSpace=BCS_BoneSpace
    translationMode=BMM_Additive
    translationSpace=BCS_BoneSpace
    alphaBoolBlend {blendOption=Linear}
    alphaScaleBiasClamp {bInterpResult=true, interpSpeedDecreasing=20, interpSpeedIncreasing=20}
    boneToModify {boneName=dyn_shoulderpad_mstr_r}
   }
   "Transform (Modify) Bone - Bone: dyn_shoulderpad_mstr_l"(Alpha=Get armRotation_l() -> (armRotation_l:double), Translation="1,1,1.5", Rotation="0,0,15", Scale="1,1,1") {
    rotationMode=BMM_Additive
    rotationSpace=BCS_BoneSpace
    translationMode=BMM_Additive
    translationSpace=BCS_BoneSpace
    alphaBoolBlend {blendOption=Linear}
    alphaScaleBiasClamp {interpSpeedIncreasing=40}
    boneToModify {boneName=dyn_shoulderpad_mstr_l}
    rotation {roll=15}
    translation=<1 1 1.5>
   }
   "Transform (Modify) Bone - Bone: dyn_shoulderpad_mstr_r"(Alpha=Get armRotation_r() -> (armRotation_r:double), Translation="0,0,1", Rotation="30,5,20", Scale="1,1,1") {
    rotationMode=BMM_Additive
    rotationSpace=BCS_BoneSpace
    translationMode=BMM_Additive
    translationSpace=BCS_BoneSpace
    alphaBoolBlend {blendOption=Linear}
    alphaScaleBiasClamp {interpSpeedIncreasing=40}
    boneToModify {boneName=dyn_shoulderpad_mstr_r}
    rotation=<rot 30 5 20>
    translation {z=1}
   }
   n2 = Get GravityOverride()
   n3 = RigidBody!(ComponentPose=Local To Component(LocalPose=Use cached pose 'preSimCache'()), Alpha, OverrideWorldGravity:Vector=@n2, ExternalForce) -> (Pose) {
    bClampLinearTranslationLimitToRefPose=true
    bOverrideWorldGravity=true
    overridePhysicsAsset=&/M_MED_Grim_Physics_Pouches
    simulationSpace=BaseBoneSpace
    alphaBoolBlend {blendOption=Linear}
    baseBoneRef {boneName=pelvis}
    componentAppliedLinearAccClamp=<3000 3000 3000>
    componentLinearAccScale=<0.4 0.4 0.4>
    componentLinearVelScale=<0.4 0.4 0.4>
    overrideWorldGravity {z=-980}
    simSpaceSettings {
     maxAngularAcceleration=1000
     maxAngularVelocity=3
     maxLinearAcceleration=700
     maxLinearVelocity=500
     velocityScaleZ=0
     worldAlpha=1
     externalLinearDragV=<1 1>
    }
   }
   n4 = RigidBody(ComponentPose=@n3, OverrideWorldGravity=@n2) {
    bClampLinearTranslationLimitToRefPose=true
    bOverrideWorldGravity=true
    overridePhysicsAsset=&/M_MED_Grim_Physics_Bandolier
    simulationSpace=BaseBoneSpace
    alphaBoolBlend {blendOption=Linear}
    baseBoneRef {boneName=spine_05}
    componentAppliedLinearAccClamp=<3000 3000 3000>
    componentLinearAccScale=<0.4 0.4 0.4>
    componentLinearVelScale=<0.4 0.4 0.4>
    overrideWorldGravity {z=-980}
    simSpaceSettings {
     maxAngularAcceleration=1000
     maxAngularVelocity=3
     maxLinearAcceleration=700
     maxLinearVelocity=500
     velocityScaleZ=0
     worldAlpha=1
     externalLinearDragV=<1 1>
    }
   }
   Output Pose(Result=Component To Local(ComponentPose=@n4)) {layerGroup=None}
   n5 = Control Rig(Source=Copy Pose From Mesh() {bUseAttachedParent=true}) {
    alphaBoolBlend {blendOption=Linear}
    controlRigAssetReference {blueprintRigClass=&/M_MED_Grim_Deform_CtrlRig_twists.M_MED_Grim_Deform_CtrlRig_twists_C}
   }
   n6 = Control Rig!(
    Source=@n5
    Alpha
    KneeRot_fr_l_114deg:Vector
    KneeTrans_fr_l_114deg:Vector="0.258,-3.591,0.009"
    KneeScale_fr_l_114deg:Vector="1,1,1"
    KneeScale_fr_l_130deg:Vector="1,1,1"
    KneeRot_fr_l_130deg:Vector="0,0,-51.2"
    KneeTrans_fr_l_130deg:Vector="0.258,-3.591,0.009"
   ) -> (Pose) {
    alphaBoolBlend {blendOption=Linear}
    controlRigAssetReference {blueprintRigClass=&/M_MED_Grim_Deform_CtrlRig_sculpters.M_MED_Grim_Deform_CtrlRig_sculpters_C}
   }
   n7 = "Constraint - dyn_bandolierBullet_01_mstr (PPPPP)"!(
    ComponentPose=Local To Component(LocalPose=@n6)
    Alpha=1
    ConstraintWeights_0=0.0643
    ConstraintWeights_1:float=0.0104
    ConstraintWeights_2:float=0.3988
    ConstraintWeights_3:float=0.3862
    ConstraintWeights_4:float=0.1403
   ) -> (Pose) {
    alphaBoolBlend {blendOption=Linear}
    boneToModify {boneName=dyn_bandolierBullet_01_mstr}
    constraintSetup [
     {targetBone {boneName=clavicle_r}}
     {targetBone {boneName=neck_02}}
     {targetBone {boneName=spine_04}}
     {targetBone {boneName=spine_05}}
     {targetBone {boneName=deform_pectoral_fr_r}}
    ]
    constraintWeights [1,1,1,1,1]
   }
   n8 = "Constraint - dyn_bandolierBullet_02_mstr (PPPPP)"!(
    ComponentPose=@n7
    Alpha=1
    ConstraintWeights_0=0.0246
    ConstraintWeights_1:float=0.0178
    ConstraintWeights_2:float=0.4841
    ConstraintWeights_3:float=0.424
    ConstraintWeights_4:float=0.0495
   ) -> (Pose) {
    alphaBoolBlend {blendOption=Linear}
    boneToModify {boneName=dyn_bandolierBullet_02_mstr}
    constraintSetup [
     {targetBone {boneName=clavicle_r}}
     {targetBone {boneName=spine_03}}
     {targetBone {boneName=spine_04}}
     {targetBone {boneName=spine_05}}
     {targetBone {boneName=deform_pectoral_fr_r}}
    ]
    constraintWeights [1,1,1,1,1]
   }
   n9 = "Constraint - dyn_bandolierBullet_03_mstr (PPPPPP)"!(
    ComponentPose=@n8
    Alpha=1
    ConstraintWeights_0=0.0245
    ConstraintWeights_1:float=0.0264
    ConstraintWeights_2:float=0.0724
    ConstraintWeights_3:float=0.4764
    ConstraintWeights_4:float=0.3881
    ConstraintWeights_5:float=0.0122
   ) -> (Pose) {
    alphaBoolBlend {blendOption=Linear}
    boneToModify {boneName=dyn_bandolierBullet_03_mstr}
    constraintSetup [
     {targetBone {boneName=deform_pectoral_fr_r}}
     {targetBone {boneName=spine_02}}
     {targetBone {boneName=spine_03}}
     {targetBone {boneName=spine_04}}
     {targetBone {boneName=spine_05}}
     {targetBone {boneName=clavicle_r}}
    ]
    constraintWeights [1,1,1,1,1,1]
   }
   n10 = "Constraint - dyn_bandolierBullet_04_mstr (PPPPP)"!(
    ComponentPose=@n9
    Alpha=1
    ConstraintWeights_0=0.0517
    ConstraintWeights_1:float=0.1175
    ConstraintWeights_2:float=0.4696
    ConstraintWeights_3:float=0.3488
    ConstraintWeights_4:float=0.0123
   ) -> (Pose) {
    alphaBoolBlend {blendOption=Linear}
    boneToModify {boneName=dyn_bandolierBullet_04_mstr}
    constraintSetup [
     {targetBone {boneName=spine_02}}
     {targetBone {boneName=spine_03}}
     {targetBone {boneName=spine_04}}
     {targetBone {boneName=spine_05}}
     {targetBone {boneName=deform_pectoral_fr_r}}
    ]
    constraintWeights [1,1,1,1,1]
   }
   n11 = "Constraint - dyn_bandolierBullet_05_mstr (PPPP)"!(
    ComponentPose=@n10
    Alpha=1
    ConstraintWeights_0=0.0789
    ConstraintWeights_1:float=0.1587
    ConstraintWeights_2:float=0.4654
    ConstraintWeights_3:float=0.297
   ) -> (Pose) {
    alphaBoolBlend {blendOption=Linear}
    boneToModify {boneName=dyn_bandolierBullet_05_mstr}
    constraintSetup [{targetBone {boneName=spine_02}}, {targetBone {boneName=spine_03}}, {targetBone {boneName=spine_04}}, {targetBone {boneName=spine_05}}]
    constraintWeights [1,1,1,1]
   }
   n12 = "Transform (Modify) Bone - Bone: dyn_shoulderpad_mstr_r"!(
    ComponentPose=@n11
    bAlphaBoolEnabled:bool=Get pistolEquip() -> (pistolEquip:bool)
    Translation="0,0,-0.8"
    Rotation="20,12,18"
    Scale="1,1,1"
   ) -> (Pose) {
    alphaInputType=Bool
    rotationMode=BMM_Additive
    rotationSpace=BCS_BoneSpace
    translationMode=BMM_Additive
    translationSpace=BCS_BoneSpace
    alphaBoolBlend {blendInTime=0.05, blendOption=Linear, blendOutTime=0.15}
    alphaScaleBiasClamp {bInterpResult=true, interpSpeedDecreasing=20, interpSpeedIncreasing=20}
    boneToModify {boneName=dyn_shoulderpad_mstr_r}
    rotation=<rot 20 12 18>
    translation {z=-0.8}
   }
   n13 = Control Rig!(Source=Component To Local(ComponentPose=@n12), Alpha) -> (Pose) {
    alphaBoolBlend {blendOption=Linear}
    controlRigAssetReference {blueprintRigClass=&/M_MED_Grim_CtrlRig.M_MED_Grim_CtrlRig_C}
   }
   Save cached pose 'preSimCache'(Pose=@n13) {cacheName=preSimCache}
  ]}
 EventGraph {nodes [
   Event BlueprintUpdateAnimation()
   > n1 = Sequence()
   > n2 = Gravity Override State Function DEPRECATED(CharacterPartAnimInstance=Self-Reference(), GameTimeInSeconds=GetGameTimeInSeconds(), GravityOverrideParams=Get Gravity Override Params())
   > Set GravityOverride(GravityOverride=@n2)
   n3 = GetOwnerSkeletalMeshComponent()
   n4 = Cast To FortPlayerAnimInstance(execute=@n1.then_1, Object=GetAnimInstance(self=@n3))
   > Set pistolEquip(pistolEquip:bool=false) -> (Output_Get:bool)
   > n5 = Switch on EFortWeaponCoreAnimation!(execute, Selection:EFortWeaponCoreAnimation Enum=Get WeaponCoreAnim(self=@n4)) -> (
    Melee:Exec
    Pistol:Exec
    Shotgun:Exec
    PaperBlueprint:Exec
    Rifle:Exec
    MeleeOneHand:Exec
    MachinePistol:Exec
    RocketLauncher:Exec
    GrenadeLauncher:Exec
    GoingCommando:Exec
    AssaultRifle:Exec
    TacticalShotgun:Exec
    SniperRifle:Exec
    TrapPlacement:Exec
    ShoulderLauncher:Exec
    AbilityDecoTool:Exec
    Crossbow:Exec
    C4:Exec
    RemoteControl:Exec
    DualWield:Exec
    AR_BullPup:Exec
    AR_ForwardGrip:Exec
    MedPackPaddles:Exec
    SMG_P90:Exec
    AR_DrumGun:Exec
    Consumable_Small:Exec
    Consumable_Large:Exec
    Balloon:Exec
    MountedTurret:Exec
    CreativeTool:Exec
    ExplosiveBow:Exec
    AshtonIndigo:Exec
    AshtonChicago:Exec
    MeleeDualWield:Exec
    Unarmed:Exec
   )
   > Set pistolEquip(pistolEquip=false)
   Branch(execute=@n5.Pistol, Condition=Get bIsTargeting())
   > Set pistolEquip(pistolEquip=true)
   Event BlueprintInitializeAnimation()
   > Set Skel_Mesh(Skel_Mesh=@n3)
  ]}
]
variables [
 armRotation_l {type=number}
 armRotation_r {type=number}
 Gravity Override Params {type=GravityOverrideParamsStruct, default {skydive_GoingUp {windOffset=<0 -600 400>}}}
 GravityOverride {type=Vector}
 pistolEquip {type=boolean}
 Skel_Mesh {type=SkeletalMeshComponent}
]
