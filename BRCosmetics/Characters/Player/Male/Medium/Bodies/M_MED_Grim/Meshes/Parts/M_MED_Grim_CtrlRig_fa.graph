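# Control Rig for a part of the medium male Grim player body that adjusts the headband's left and right master bones as the head moves.
# A full-weight transform modify first rotates those bones.
# A spherical pose reader on the head, relative to spine_04, then weights a second modify with roughly opposite rotations, countering the first offset.
# A head_alt-to-neck_01 parent constraint is unwired, and the RBOFF variable is unused.
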ControlRigBlueprint /BRCosmetics/Characters/Player/Male/Medium/Bodies/M_MED_Grim/Meshes/Parts/M_MED_Grim_CtrlRig_fa
graphs [
 RigVMFunctionLibrary {links [], nodes []}
 RigVMModel {nodes [
   RigUnit_BeginExecution = &RigUnit_BeginExecution()
   &RigUnit_ModifyTransforms(
    ItemToModify="((Item=(Type=Bone,Name=\"dyn_headband_mstr_l\"),Transform=(Rotation=(X=0,Y=0,Z=-0.3439,W=0.939),Translation=(X=0,Y=0,Z=0),Scale3D=(X=1,Y=1,Z=1))),(Item=(Type=Bone,Name=\"dyn_headband_mstr_r\"),Transform=(Rotation=(X=0,Y=0,Z=-0.2376,W=0.9714),Translation=(X=0,Y ...(truncated)"
    Weight=1
    WeightMinimum=0
    ExecutePin=@RigUnit_BeginExecution.ExecutePin
   )
   > n1 = &RigUnit_SphericalPoseReader(
    ActiveRegionScaleFactors=(PositiveWidth=3,NegativeWidth=0.3,PositiveHeight=6,NegativeHeight=6)
    ActiveRegionSize=0.1
    Debug=(bDrawDebug=True,DebugScale=25,DebugSegments=20,DebugThickness=0.25)
    DriverAxis=(X=1,Y=0,Z=0)
    DriverItem=(Type=Bone,Name="head")
    FalloffRegionScaleFactors=(PositiveWidth=-3,NegativeWidth=1,PositiveHeight=0,NegativeHeight=0)
    FalloffSize=0.25
    OptionalParentItem=(Type=Bone,Name="spine_04")
    RotationOffset=(X=90,Y=0,Z=-140)
   )
   > &RigUnit_ModifyTransforms(
    ItemToModify="((Item=(Type=Bone,Name=\"dyn_headband_mstr_l\"),Transform=(Rotation=(X=0,Y=0,Z=0.342,W=0.9397),Translation=(X=0,Y=0,Z=0),Scale3D=(X=1,Y=1,Z=1))),(Item=(Type=Bone,Name=\"dyn_headband_mstr_r\"),Transform=(Rotation=(X=0,Y=0,Z=0.2334,W=0.9724),Translation=(X=0,Y=0 ...(truncated)"
    Weight=@n1.OutputParam
   )
   &RigUnit_ParentConstraint(
    AdvancedSettings=(InterpolationType=Average,RotationOrderForFilter=XZY)
    Child=(Type=Bone,Name="head_alt")
    ExecutePin=()
    Parents=((Item=(Type=Bone,Name="neck_01"),Weight=1))
    Weight=1
   )
  ]}
]
hierarchy [
 root {loc=<0 0 0>, rot=<0 0 0 1>}
 pelvis {parent=root, loc=<0 -0.7071 93.1191>, rot=<0.0127 0.707 -0.0127 -0.707>}
 spine_01 {parent=pelvis, loc=<6.7494 -0.6388 0>, rot=<0 0 0.034 -0.9994>}
 spine_02 {parent=spine_01, loc=<7.2518 -0.035 0>, rot=<0 0 0.0238 -0.9997>}
 spine_03 {parent=spine_02, loc=<6.0104 0 0>, rot=<0 0 0.0149 -0.9999>}
 spine_04 {parent=spine_03, loc=<8.1024 0 0>, rot=<0 0 -0.0132 -0.9999>}
 spine_05 {parent=spine_04, loc=<10.0146 0 0>, rot=<0 0 -0.0015 -1>}
 neck_01 {parent=spine_05, loc=<11.3138 0.1529 0>, rot=<0 0 -0.2797 -0.9601>}
 neck_02 {parent=neck_01, loc=<4.5627 0 0>, rot=<0 0 0.0504 -0.9987>}
 head {parent=neck_02, loc=<5.3914 0 0>, rot=<0 0 0.115 -0.9934>}
 head_alt {parent=neck_01, loc=<9.9266 -0.5431 0>, rot=<0 0 0.1649 -0.9863>}
 dyn_headband {parent=head_alt, loc=<3.7919 -6.1468 0.0926>, rot=<0.5021 -0.4979 0.4979 -0.5021>}
 dyn_headband_mstr_l {parent=dyn_headband, loc=<0.2585 0.9205 -0.0514>, rot=<-0.7062 -0.0363 -0.0005 -0.7071>}
 dyn_headband_01_l {parent=dyn_headband_mstr_l, loc=<0 0 0>, rot=<0 0 0 1>}
 dyn_headband_02_l {parent=dyn_headband_01_l, loc=<9.9892 0 0>, rot=<-0.0006 0.0137 -0.0039 -0.9999>}
 dyn_headband_03_l {parent=dyn_headband_02_l, loc=<10.7766 0 0>, rot=<-0.0001 0.0028 -0.0175 -0.9998>}
 dyn_headband_04_l {parent=dyn_headband_03_l, loc=<12.5842 0 0>, rot=<0 0 0 1>}
 dyn_headband_mstr_r {parent=dyn_headband, loc=<0.583 -0.5457 -0.2944>, rot=<-0.7071 -0.0027 0.1552 -0.6899>}
 dyn_headband_01_r {parent=dyn_headband_mstr_r, loc=<0 0 0>, rot=<0 0 0 1>}
 dyn_headband_02_r {parent=dyn_headband_01_r, loc=<6.4371 0 0>, rot=<0.0058 -0.0266 -0.0092 -0.9996>}
 dyn_headband_03_r {parent=dyn_headband_02_r, loc=<9.3765 0 0>, rot=<0 -0.0001 -0.0077 -1>}
 dyn_headband_04_r {parent=dyn_headband_03_r, loc=<10.2041 0 0>, rot=<0 0 0 1>}
 sword_r {parent=spine_05, loc=<19.8164 -15.3158 13.4676>, rot=<-0.0138 0.3438 -0.0377 -0.9382>}
 sword_l {parent=spine_05, loc=<19.8094 -16.6227 -13.5975>, rot=<-0.6524 -0.2728 0.2197 -0.6721>}
]
variables [RBOFF {type=double, default=0}]
