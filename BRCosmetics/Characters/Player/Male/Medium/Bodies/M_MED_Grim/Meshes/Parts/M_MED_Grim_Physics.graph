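# Physics asset for the Grim medium male player body, giving the dangling left and right headband tails secondary physics motion.
# Each tail is a four-capsule chain hung from a master bone, joined through the headband root to capsules on neck_01 and spine_05.
# Tail joints limit swing and twist, with tighter second-axis swing on the first two segments of each side, while root, master and neck links rotate freely.
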PhysicsAsset /BRCosmetics/Characters/Player/Male/Medium/Bodies/M_MED_Grim/Meshes/Parts/M_MED_Grim_Physics
bodies {
 dyn_headband [{shapeType=Capsule, length=0.505, radius=0.505}]
 dyn_headband_01_l [{shapeType=Capsule, length=2.75, radius=1.5, center=<5.5133 0 0.0002>, rotation=<rot 90 0 0>}]
 dyn_headband_01_r [{shapeType=Capsule, length=2, radius=1.5, center=<4 0 0>, rotation=<rot 90 0 0>}]
 dyn_headband_02_l [{shapeType=Capsule, length=7, radius=1.5, center=<4 0 0>, rotation=<rot 90 0 0>}]
 dyn_headband_02_r [{shapeType=Capsule, length=7, radius=1.5, center=<4 0 0>, rotation=<rot 90 0 0>}]
 dyn_headband_03_l [{shapeType=Capsule, length=6.75, radius=1.75, center=<5.2384 0 -0.0001>, rotation=<rot 90 0 0>}]
 dyn_headband_03_r [{shapeType=Capsule, length=6.5, radius=2, center=<4 0 0>, rotation=<rot 90 0 0>}]
 dyn_headband_04_l [{shapeType=Capsule, length=6, radius=2.5, center=<4 0 0>, rotation=<rot 90 0 0>}]
 dyn_headband_04_r [{shapeType=Capsule, length=6, radius=2.5, center=<4 0 0>, rotation=<rot 90 0 0>}]
 dyn_headband_mstr_l [{shapeType=Capsule, length=0.505, radius=0.505}]
 dyn_headband_mstr_r [{shapeType=Capsule, length=0.505, radius=0.505}]
 neck_01 [{shapeType=Capsule, length=14.5, radius=10, center=<-0.1047 -1.9419 0>}]
 spine_05 [{shapeType=Capsule, length=26, radius=16.75, center=<0 -0.4656 0>}]
}
constraints [
 {bone1Name=dyn_headband_01_l, bone2Name=dyn_headband_mstr_l, swing1LimitDegrees=45, swing2LimitDegrees=5, twistLimitDegrees=15}
 {bone1Name=dyn_headband_02_l, bone2Name=dyn_headband_01_l, swing1LimitDegrees=45, swing2LimitDegrees=5, twistLimitDegrees=15}
 {bone1Name=dyn_headband_03_l, bone2Name=dyn_headband_02_l, swing1LimitDegrees=45, swing2LimitDegrees=45, twistLimitDegrees=15}
 {bone1Name=dyn_headband_04_l, bone2Name=dyn_headband_03_l, swing1LimitDegrees=45, swing2LimitDegrees=45, twistLimitDegrees=15}
 {bone1Name=dyn_headband_01_r, bone2Name=dyn_headband_mstr_r, swing1LimitDegrees=45, swing2LimitDegrees=5, twistLimitDegrees=15}
 {bone1Name=dyn_headband_02_r, bone2Name=dyn_headband_01_r, swing1LimitDegrees=45, swing2LimitDegrees=5, twistLimitDegrees=15}
 {bone1Name=dyn_headband_03_r, bone2Name=dyn_headband_02_r, swing1LimitDegrees=45, swing2LimitDegrees=45, twistLimitDegrees=15}
 {bone1Name=dyn_headband_04_r, bone2Name=dyn_headband_03_r, swing1LimitDegrees=45, swing2LimitDegrees=45, twistLimitDegrees=15}
 {
  bone1Name=dyn_headband
  bone2Name=neck_01
  swing1LimitDegrees=45
  swing1Motion=Free
  swing2LimitDegrees=45
  swing2Motion=Free
  twistLimitDegrees=45
  twistMotion=Free
 }
 {
  bone1Name=dyn_headband_mstr_r
  bone2Name=dyn_headband
  swing1LimitDegrees=45
  swing1Motion=Free
  swing2LimitDegrees=45
  swing2Motion=Free
  twistLimitDegrees=45
  twistMotion=Free
 }
 {
  bone1Name=dyn_headband_mstr_l
  bone2Name=dyn_headband
  swing1LimitDegrees=45
  swing1Motion=Free
  swing2LimitDegrees=45
  swing2Motion=Free
  twistLimitDegrees=45
  twistMotion=Free
 }
 {
  bone1Name=neck_01
  bone2Name=spine_05
  swing1LimitDegrees=45
  swing1Motion=Free
  swing2LimitDegrees=45
  swing2Motion=Free
  twistLimitDegrees=45
  twistMotion=Free
 }
]
