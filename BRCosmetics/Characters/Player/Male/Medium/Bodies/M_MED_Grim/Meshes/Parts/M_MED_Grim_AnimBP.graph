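# Post-copy AnimBlueprint for the Grim cosmetic body part, adding secondary motion to the pose copied from the attached parent mesh.
# A Control Rig runs when rigid bodies are disabled; otherwise head_alt follows head rotation and headband bones get noise-driven sway.
# One of two RigidBody physics assets runs, chosen by the GrimPack meta tag, with state-based gravity override.
# Skydiving and weapon events toggle a material hide parameter.
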AnimBlueprint /BRCosmetics/Characters/Player/Male/Medium/Bodies/M_MED_Grim/Meshes/Parts/M_MED_Grim_AnimBP
target_skeleton=&/M_MED_Grim_Skeleton
graphs [
 AnimGraph {nodes [
   n1 = Control Rig(Source=Copy Pose From Mesh() {bUseAttachedParent=true}, Alpha=Get RB_Off() -> (RB_Off:double)) {
    alphaBoolBlend {blendOption=Linear}
    controlRigAssetReference {blueprintRigClass=&/M_MED_Grim_CtrlRig_fa.M_MED_Grim_CtrlRig_fa_C}
   }
   n2 = BreakRotator(InRot=Get head rot() -> (head rot:Rotator))
   n3 = "Transform (Modify) Bone - Bone: head_alt"(
    ComponentPose=Local To Component(LocalPose=@n1)
    Alpha=Get RB_On() -> (RB_On:double)
    Translation="0,0,0"
    Rotation=MakeRotator(Roll=@n2, Pitch=@n2.Pitch, Yaw=float * float(A=@n2.Yaw, B=0.5))
    Scale="1,1,1"
   ) {rotationMode=BMM_Replace, rotationSpace=BCS_BoneSpace, alphaBoolBlend {blendOption=Linear}, boneToModify {boneName=head_alt}}
   n4 = Get bRigidBodyEnabled()
   n5 = "Transform (Modify) Bone - Bone: dyn_headband_mstr_l"!(ComponentPose=@n3, bAlphaBoolEnabled:bool=@n4, Translation="0,0,0", Rotation=Get HairRotator_01() -> (HairRotator_01:Rotator), Scale="1,1,1") -> (Pose) {
    alphaInputType=Bool
    rotationMode=BMM_Additive
    rotationSpace=BCS_BoneSpace
    alphaBoolBlend {blendOption=Linear}
    boneToModify {boneName=dyn_headband_mstr_l}
   }
   n6 = "Transform (Modify) Bone - Bone: dyn_headband_mstr_r"!(ComponentPose=@n5, bAlphaBoolEnabled:bool=@n4, Translation="0,0,0", Rotation=Get HairRotator_03() -> (HairRotator_03:Rotator), Scale="1,1,1") -> (Pose) {
    alphaInputType=Bool
    rotationMode=BMM_Additive
    rotationSpace=BCS_BoneSpace
    alphaBoolBlend {blendOption=Linear}
    boneToModify {boneName=dyn_headband_mstr_r}
   }
   n7 = Get grimPackEquipped() -> (grimPackEquipped:bool)
   n8 = Get Gravity Override()
   n9 = RigidBody!(ComponentPose=@n6, bAlphaBoolEnabled:bool=NOT Boolean(A=@n7), OverrideWorldGravity:Vector=@n8, ExternalForce) -> (Pose) {
    alphaInputType=Bool
    bClampLinearTranslationLimitToRefPose=true
    bOverrideWorldGravity=true
    overridePhysicsAsset=&/M_MED_Grim_Physics
    simulationSpace=BaseBoneSpace
    alphaBoolBlend {blendOption=Linear}
    baseBoneRef {boneName=neck_01}
    componentAppliedLinearAccClamp=<3000 3000 3000>
    componentLinearAccScale {x=0.1}
    componentLinearVelScale {x=0.1}
    overrideWorldGravity {z=-980}
    simSpaceSettings {
     maxAngularAcceleration=1000
     maxAngularVelocity=3
     maxLinearAcceleration=700
     maxLinearVelocity=500
     velocityScaleZ=0
     worldAlpha=1
     externalLinearDragV=<0.2 0.2>
    }
   }
   n10 = RigidBody(ComponentPose=@n9, bAlphaBoolEnabled=@n7, OverrideWorldGravity=@n8) {
    alphaInputType=Bool
    bClampLinearTranslationLimitToRefPose=true
    bOverrideWorldGravity=true
    overridePhysicsAsset=&/M_MED_Grim_Physics_Pack
    simulationSpace=BaseBoneSpace
    alphaBoolBlend {blendOption=Linear}
    baseBoneRef {boneName=neck_01}
    componentAppliedLinearAccClamp=<3000 3000 3000>
    componentLinearAccScale {x=0.1}
    componentLinearVelScale {x=0.1}
    overrideWorldGravity {z=-980}
    simSpaceSettings {
     maxAngularAcceleration=1000
     maxAngularVelocity=3
     maxLinearAcceleration=700
     maxLinearVelocity=500
     velocityScaleZ=0
     worldAlpha=1
     externalLinearDragV=<0.2 0.2>
    }
   }
   Output Pose(Result=Component To Local(ComponentPose=@n10)) {layerGroup=None}
  ]}
 EventGraph {nodes [
   Event BlueprintUpdateAnimation()
   > n1 = Sequence!(execute) -> (then_0, then_1, then_2:Exec)
   > n2 = Gravity Override State Function DEPRECATED(CharacterPartAnimInstance=Self-Reference(), GameTimeInSeconds=0, GravityOverrideParams=Get Gravity Override Params())
   > Set Gravity Override(Gravity Override=@n2)
   n3 = Get bRigidBodyEnabled()
   > Set RB_Off(RB_Off:double="To Float (Boolean)"(InBool=NOT Boolean(A=@n3))) -> (Output_Get:double)
   > Set RB_On(RB_On:double="To Float (Boolean)"(InBool=Get bRigidBodyEnabled())) -> (Output_Get:double)
   n4 = GetGameTimeInSeconds()
   n5 = RandomRotatorFunction!(
    execute=@n1.then_2
    GameTimeInSeconds:double=@n4
    PerlinScale:double=1
    Frequency:Rotator="2.2,1.5, 0"
    Amplitude:Rotator="15,10,0"
    Offset:Rotator="0,-2,0"
    TimeOffset:double=0
   ) -> (then, RandomRotator:Rotator)
   > Set HairRotator_01(HairRotator_01:Rotator=@n5) -> (Output_Get:Rotator)
   > n6 = RandomRotatorFunction(GameTimeInSeconds=@n4, PerlinScale=1, Frequency="1.8,2.7, 0", Amplitude="10,15,0", Offset="0,3,0", TimeOffset=0)
   > Set HairRotator_03(HairRotator_03:Rotator=@n6) -> (Output_Get:Rotator)
   n7 = Get Skel_Mesh()
   Set head rot(
    execute=@n1.then_1
    head rot:Rotator=BreakTransform(InTransform=GetDeltaTransformFromRefPose(self=@n7, BoneName=head, BaseName=neck_02)).Rotation
   ) -> (Output_Get:Rotator)
   Event BlueprintInitializeAnimation()
   > Set Skel_Mesh(Skel_Mesh=GetOwnerSkeletalMeshComponent())
   Event OnWeaponChanged()
   > n8 = Get Dynamic Material Instance(SkeletalMeshComponent=GetPartSkeletalMeshComponent())
   > n9 = Branch()
   > n10 = Branch(Condition=Get bIsSkyDiving())
   SetScalarParameterValue(execute=[@n11.then, @n10.then, @n12.else, @n9.else], self=@n8, ParameterName=Hide Element 01, Value=0)
   SetScalarParameterValue(execute=[@n10.else, @n12.then], self=@n8, ParameterName=Hide Element 01, Value=1)
   n11 = Event OnBeginSkydiving()
   Event OnEndSkydiving()
   > n12 = Branch()
   Event OnCharacterCustomizationLoaded()
   > Set grimPackEquipped(grimPackEquipped:bool=CharacterItemDefinitionHasAnyMetaTag(Tags=Make Array("[0]"=(TagName="Cosmetics.MetaTags.GrimPack")))) -> (Output_Get:bool)
   GetGameTimeInSeconds()
  ]}
]
variables [
 Gravity Override {type=Vector}
 Gravity Override Params {
  type=GravityOverrideParamsStruct
  default {
   onGround_Moving {gravityOverride=<0 -50 -15>, windAmplitude=<100 -50 200>, windFrequency=<22.5 150 45>, windOffset=<0 -100 -50>}
   onGround_Standing {gravityOverride=<0 -50 -15>, windAmplitude=<100 -50 200>, windFrequency=<22.5 150 45>, windOffset=<0 -100 -50>}
   skydive_GoingUp {gravityOverride=<0 -300 -100>}
   skydive_Parachuting {gravityOverride=<0 -800 -300>}
   swimming {gravityOverride=<0 0 0>}
  }
 }
 grimPackEquipped {type=boolean}
 HairRotator_01 {type=Rotator}
 HairRotator_03 {type=Rotator}
 head rot {type=Rotator}
 RB_Off {type=number}
 RB_On {type=number}
 Skel_Mesh {type=SkeletalMeshComponent}
]
